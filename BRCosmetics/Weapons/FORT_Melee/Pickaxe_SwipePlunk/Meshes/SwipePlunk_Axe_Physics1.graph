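# Physics asset for the SwipePlunk two-handed pickaxe, giving its handle, skull and spine bones collision bodies and jointed limits.
# Each of the four bones has one capsule.
# Spine_base and Spine_tail both constrain to Melee_TwoHanded_Handle, and Skull hangs off Spine_base, forming two branches from the handle.
# Angular limits on swing and twist bound each joint, with the tail given the widest swing.
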PhysicsAsset /BRCosmetics/Weapons/FORT_Melee/Pickaxe_SwipePlunk/Meshes/SwipePlunk_Axe_Physics1
bodies {
 Melee_TwoHanded_Handle [{shapeType=Capsule, length=4.3382, radius=2.5678, center=<0 0.1122 -1.7023>}]
 Skull [{shapeType=Capsule, length=2.3091, radius=9.9914, shapeName=Skull_capsule, center=<5.845 -0.1402 0.8443>}]
 Spine_base [{shapeType=Capsule, length=12.1293, radius=2.9005, center=<9.707 0 0>, rotation=<rot -90 -0.0001 0>}]
 Spine_tail [{shapeType=Capsule, length=9.2789, radius=3.1825, shapeName=Spine_tail_capsule, center=<11.8098 0.0001 -1.8532>, rotation=<rot 90 0.0012 0>}]
}
constraints [
 {bone1Name=Spine_base, bone2Name=Melee_TwoHanded_Handle, swing1LimitDegrees=10, swing2LimitDegrees=25, twistLimitDegrees=10}
 {bone1Name=Spine_tail, bone2Name=Melee_TwoHanded_Handle, swing1LimitDegrees=40, swing2LimitDegrees=40, twistLimitDegrees=15}
 {bone1Name=Skull, bone2Name=Spine_base, swing1LimitDegrees=20, swing2LimitDegrees=31, twistLimitDegrees=15}
]
