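# Control Rig for the SwipePlunk pickaxe that adds simulated physics motion to its skeleton, which has spine, skull and tail bones.
# Construction adds a world-space solver on b_AR_Root with clamped simulation-space motion, then builds bodies and joints from the SwipePlunk_Axe_Physics1 asset.
# The forward solve steps the stored solver at partial alpha with a delta-time override, leaving debug visualization enabled.
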ControlRigBlueprint /BRCosmetics/Weapons/FORT_Melee/Pickaxe_SwipePlunk/Meshes/SwipePlunk_Axe_CtrlRig
graphs [
 RigVMFunctionLibrary {links [], nodes []}
 RigVMModel {nodes [
   n1 = &RigUnit_PrepareForExecution()
   n2 = /Script/ControlRigPhysics.RigUnit_AddPhysicsSolver(
    Owner=(Type=Bone,Name="b_AR_Root")
    SimulationSpaceSettings="(SpaceMovementAmount=0.5,VelocityScaleZ=0.5,bClampLinearVelocity=true,MaxLinearVelocity=8,bClampAngularVelocity=true,MaxAngularVelocity=8,bClampLinearAcceleration=true,MaxLinearAcceleration=40,bClampAngularAcceleration=true,MaxAngularAcceleration=40,Linear ...(truncated)"
    SolverSettings="(bAutomaticallyAddPhysicsBodyComponents=True,SimulationSpace=World,CollisionSpace=World,SpaceBone=(Type=Bone,Name=\"b_AR_Root\"),Collision=(Material=(Friction=1,Restitution=1,FrictionCombineMode=Multiply,RestitutionCombineMode=Multiply)),Gravity=(X=0,Y=0,Z=0 ...(truncated)"
    ExecutePin=@n1.ExecutePin
   ) {is=struct RigVMUnitNode}
   > RigVMVariableNode(Value=@n2.PhysicsSolverComponentKey) {variable=AddPhysicsSolver_PhysicsSolverComponentKey_0}
   > /Script/ControlRigPhysics.RigUnit_HierarchyInstantiateFromPhysicsAsset(
    BonesToUse=()
    ConstraintProfileName=None
    ParentSpaceControlData="(bEnabled=true,LinearStrength=0,LinearDampingRatio=1,LinearExtraDamping=0,MaxForce=0,AngularStrength=0,AngularDampingRatio=1,AngularExtraDamping=0,MaxTorque=0,LinearTargetVelocityMultiplier=1,AngularTargetVelocityMultiplier=1,CustomControlPoint=(X=0,Y=0,Z= ...(truncated)"
    PhysicsAsset=&/SwipePlunk_Axe_Physics1
    SimSpaceControlData="(bEnabled=true,LinearStrength=0,LinearDampingRatio=1,LinearExtraDamping=0,MaxForce=0,AngularStrength=0,AngularDampingRatio=1,AngularExtraDamping=0,MaxTorque=0,LinearTargetVelocityMultiplier=1,AngularTargetVelocityMultiplier=1,CustomControlPoint=(X=0,Y=0,Z= ...(truncated)"
    bAddParentSpaceControl=False
    bAddSimSpaceControl=false
    bEnableDrives=true
    bEnableJoints=true
    bMakeHelperBodies=True
    bUseAutomaticSolver=true
    PhysicsSolverComponentKey=@n2.PhysicsSolverComponentKey
   ) {is=struct RigVMUnitNode}
   RigUnit_BeginExecution = &RigUnit_BeginExecution()
   /Script/ControlRigPhysics.RigUnit_StepPhysicsSolver(
    Alpha=0.5
    DeltaTimeOverride=0.1
    SimulationSpaceDeltaTimeOverride=0
    VisualizationSettings=(bEnableVisualization=true,LineThickness=1,ShapeSize=1,ShapeDetail=16,bShowBodies=true,bShowCentreOfMass=False,bShowJoints=True,bShowControls=False,bShowWorldObjects=True,bShowWorldOverlapBox=False,bShowActiveContacts=True,bShowInactiveContacts=False)
    ExecutePin=@RigUnit_BeginExecution.ExecutePin
    PhysicsSolverComponentKey=RigVMVariableNode() {variable=AddPhysicsSolver_PhysicsSolverComponentKey_0}.Value
   ) {is=struct RigVMUnitNode}
  ]}
]
hierarchy [
 b_AR_Root {loc=<0 0 0>, rot=<0 0 0 1>}
 Melee_TwoHanded_Handle {parent=b_AR_Root, loc=<0 0 0>, rot=<0 0 0.7071 -0.7071>}
 Spine_base {parent=Melee_TwoHanded_Handle, loc=<0 -1.0496 4.9074>, rot=<0.6093 -0.6093 0.3588 0.3588>}
 Skull {parent=Spine_base, loc=<21.7528 0 0>, rot=<0 -0.0942 0 -0.9956>}
 Spine_tail {parent=Melee_TwoHanded_Handle, loc=<0 -0.8209 -8.4529>, rot=<0.3764 -0.3764 -0.5986 -0.5986>}
]
variables [
 AddPhysicsSolver_PhysicsSolverComponentKey {type=FRigComponentKey, default=(ElementKey=(Type=None,Name=""),Name="")}
 AddPhysicsSolver_PhysicsSolverComponentKey_0 {type=FRigComponentKey, default=(ElementKey=(Type=None,Name=""),Name="")}
 AddPhysicsSolver_SimulationSpaceSettings {
  type=FRigPhysicsSimulationSpaceSettings
  default="(SpaceMovementAmount=1,VelocityScaleZ=1,bClampLinearVelocity=True,MaxLinearVelocity=10000,bClampAngularVelocity=True,MaxAngularVelocity=10000,bClampLinearAcceleration=True,MaxLinearAcceleration=10000,bClampAngularAcceleration=True,MaxAngularAcceleration=10 ...(truncated)"
 }
 AddPhysicsSolver_SolverSettings {
  type=FRigPhysicsSolverSettings
  default="(bAutomaticallyAddPhysicsBodyComponents=True,SimulationSpace=Component,CollisionSpace=Component,SpaceBone=(Type=None,Name=\"\"),Collision=(Boxes=,Spheres=,Capsules=,Convexes=,Material=(Friction=1,Restitution=1,FrictionCombineMode=Multiply,RestitutionCombineM ...(truncated)"
 }
]
