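# Animation Blueprint for the SwipePlunk pickaxe that adds secondary motion to its skeleton.
# The cached input pose feeds a bool blend on bIsInFrontEnd, using a Control Rig in the front end and RigidBody physics based at Spine_base otherwise.
# An AnimDynamics chain (skull, spine, tail) and three other RigidBody nodes are left unconnected to the output.
# The update event sets SimlationSpace, which no node here uses.
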AnimBlueprint /BRCosmetics/Weapons/FORT_Melee/Pickaxe_SwipePlunk/Meshes/SwipePlunk_Axe_AnimBP
target_skeleton=&/SwipePlunk_Axe_Skeleton
graphs [
 AnimGraph {nodes [
   n1 = Anim Dynamics - Bone: Skull(Alpha=1, ExternalForce="0,0,0") {
    angularDampingOverride=0.9999
    angularSpringConstant=600
    bAngularSpring=true
    bGravityOverrideInSimSpace=true
    bOverrideAngularDamping=true
    bUseGravityOverride=true
    simulationSpace=World
    boundBone {boneName=Skull}
    componentAppliedLinearAccClamp=<1500 1500 1500>
    componentLinearAccScale=<0.1 0.1 0.1>
    componentLinearVelScale=<0.1 0.1 0.1>
    physicsBodyDefinitions [{
      boundBone {boneName=Skull}
      boxExtents=<15 15 15>
      constraintSetup {angularLimitsMax=<10 20 25>, angularLimitsMin=<-10 -5 -25>, angularTarget=<1 0 0>}
      localJointOffset=<0.2 0 0>
     }]
    simSpaceSettings {maxAngularAcceleration=15, maxAngularVelocity=3, simSpaceAngularAlpha=0.05}
   }
   n2 = Anim Dynamics - Bone: Spine_base(ComponentPose=@n1, Alpha=1, ExternalForce="0,0,0") {
    angularDampingOverride=0.9999
    angularSpringConstant=600
    bAngularSpring=true
    bGravityOverrideInSimSpace=true
    bOverrideAngularDamping=true
    bUseGravityOverride=true
    simulationSpace=World
    boundBone {boneName=Spine_base}
    componentAppliedLinearAccClamp=<1500 1500 1500>
    componentLinearAccScale=<0.1 0.1 0.1>
    componentLinearVelScale=<0.1 0.1 0.1>
    gravityOverride {z=-300}
    physicsBodyDefinitions [{
      boundBone {boneName=Spine_base}
      boxExtents=<15 15 15>
      constraintSetup {angularLimitsMax=<10 20 30>, angularLimitsMin=<-10 -5 -30>, angularTarget=<1 0 0>}
      localJointOffset=<0.2 0 0>
     }]
    simSpaceSettings {maxAngularAcceleration=15, maxAngularVelocity=3, simSpaceAngularAlpha=0.05}
   }
   Anim Dynamics - Bone: Spine_tail(ComponentPose=@n2, Alpha=1, ExternalForce="0,0,0") {
    angularDampingOverride=0.9999
    angularSpringConstant=600
    bAngularSpring=true
    bGravityOverrideInSimSpace=true
    bOverrideAngularDamping=true
    bUseGravityOverride=true
    simulationSpace=World
    boundBone {boneName=Spine_tail}
    componentAppliedLinearAccClamp=<1500 1500 1500>
    componentLinearAccScale=<0.1 0.1 0.1>
    componentLinearVelScale=<0.1 0.1 0.1>
    gravityOverride {z=-600}
    physicsBodyDefinitions [{
      boundBone {boneName=Spine_tail}
      boxExtents=<15 15 15>
      constraintSetup {angularLimitsMax=<10 5 30>, angularLimitsMin=<-10 -20 -30>, angularTarget=<1 0 0>}
      localJointOffset=<0.2 0 0>
     }]
    simSpaceSettings {maxAngularAcceleration=15, maxAngularVelocity=3, simSpaceAngularAlpha=0.05}
   }
   RigidBody() {
    bClampLinearTranslationLimitToRefPose=true
    bOverrideWorldGravity=true
    overridePhysicsAsset=&/SwipePlunk_Axe_Physics_Front
    simulationSpace=BaseBoneSpace
    alphaBoolBlend {blendOption=Linear}
    baseBoneRef {boneName=Spine_base}
    componentAppliedLinearAccClamp=<10 10 10>
    componentLinearAccScale=<0.1 0.1 0.1>
    componentLinearVelScale=<0.1 0.1 0.1>
    simSpaceSettings {maxAngularAcceleration=30, maxAngularVelocity=3, maxLinearAcceleration=15, maxLinearVelocity=3, velocityScaleZ=0}
   }
   RigidBody() {
    bOverrideWorldGravity=true
    overridePhysicsAsset=&/SwipePlunk_Axe_Physics
    alphaBoolBlend {blendOption=Linear}
    baseBoneRef {boneName=b_AR_Root}
    overrideWorldGravity {z=-980}
    simSpaceSettings {maxAngularAcceleration=10, maxAngularVelocity=3, maxLinearAcceleration=10, maxLinearVelocity=10, velocityScaleZ=0.2, worldAlpha=1}
   }
   RigidBody(Alpha=0.3) {
    bClampLinearTranslationLimitToRefPose=true
    bOverrideWorldGravity=true
    overridePhysicsAsset=&SwipePlunk_Axe_Physics_Front
    simulationSpace=WorldSpace
    alphaBoolBlend {blendOption=Linear}
    baseBoneRef {boneName=Spine_base}
    componentAppliedLinearAccClamp=<1200 1200 1200>
    componentLinearAccScale=<0.2 0.2 0.2>
    componentLinearVelScale=<0.2 0.2 0.2>
    overrideWorldGravity {z=-980}
    simSpaceSettings {maxAngularAcceleration=20, maxAngularVelocity=3, maxLinearAcceleration=15, maxLinearVelocity=4, velocityScaleZ=0.2, worldAlpha=0.2}
   }
   n3 = RigidBody(ComponentPose=Local To Component(LocalPose=Use cached pose 'Main pose'())) {
    bClampLinearTranslationLimitToRefPose=true
    bOverrideWorldGravity=true
    overridePhysicsAsset=&/SwipePlunk_Axe_Physics1
    simulationSpace=BaseBoneSpace
    alphaBoolBlend {blendOption=Linear}
    baseBoneRef {boneName=Spine_base}
    overrideWorldGravity {z=-980}
    simSpaceSettings {
     dampingAlpha=0
     maxAngularAcceleration=32
     maxAngularVelocity=8
     maxLinearAcceleration=450
     maxLinearVelocity=10
     velocityScaleZ=0.7
     worldAlpha=1
    }
   }
   n4 = Control Rig(Source=Use cached pose 'Main pose'()) {controlRigAssetReference {blueprintRigClass=&/SwipePlunk_Axe_CtrlRig.SwipePlunk_Axe_CtrlRig_C}}
   Output Pose(Result=Blend Poses by bool(BlendPose_0=@n4, BlendPose_1=Component To Local(ComponentPose=@n3), bActiveValue=Get bIsInFrontEnd())) {layerGroup=None}
   AnimGraphNode_SubInput_0 = Input Pose()
   Save cached pose 'Main pose'(Pose=@AnimGraphNode_SubInput_0) {cacheName=Main pose}
  ]}
 EventGraph {nodes [
   Event BlueprintUpdateAnimation()
   > n1 = Branch(Condition=Get bIsInFrontEnd())
   > Set SimlationSpace(SimlationSpace:AnimPhysSimSpaceType Enum=Component) -> (Output_Get:AnimPhysSimSpaceType Enum)
   Set SimlationSpace(execute=@n1.else)
   TryGetPawnOwner()
  ]}
]
variables [SimlationSpace {type=AnimPhysSimSpaceType}]
